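# Physics asset for the ScotchPeek two-handed pickaxe mesh that defines bodies and joints for two dangling straps, A and B.
# A capsule on the handle anchors two small base bodies through free angular joints, and each base leads a chain of three strap capsules.
# Chain joints allow almost no twist and swing mostly on one axis, with a wide arc at the first link and tighter limits further down.
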PhysicsAsset /BRCosmetics/Weapons/FORT_Melee/Pickaxe_ScotchPeek/Meshes/ScotchPeek_Axe_Physics
bodies {
 Melee_TwoHanded_Handle [{
   shapeType=Capsule
   length=20
   radius=2
   shapeName=Melee_TwoHanded_Handle_capsule
   center=<-0.2346 0.4085 -6.0796>
   rotation=<rot -4.4194 -90.0001 0.0006>
  }]
 base_strap_A [{shapeType=Capsule, length=0.505, radius=0.505, shapeName=base_strap_A_capsule, rotation=<rot 0 -90 0>}]
 base_strap_B [{shapeType=Capsule, length=0.505, radius=0.505, shapeName=base_strap_B_capsule, rotation=<rot 0 -90 0>}]
 dyn_strap_A_01 [{shapeType=Capsule, length=3, radius=1, shapeName=dyn_strap_A_01_capsule, center=<1 0 0>, rotation=<rot 90 0 0>}]
 dyn_strap_A_02 [{shapeType=Capsule, length=3, radius=1, shapeName=dyn_strap_A_02_capsule, center=<1 0 0>, rotation=<rot 90 0 0>}]
 dyn_strap_A_03 [{shapeType=Capsule, length=3, radius=1, shapeName=dyn_strap_A_03_capsule, center=<1.5 0 0>, rotation=<rot 90 0 0>}]
 dyn_strap_B_01 [{shapeType=Capsule, length=3, radius=1, shapeName=dyn_strap_B_01_capsule, center=<1 0 0>, rotation=<rot 90 0 0>}]
 dyn_strap_B_02 [{shapeType=Capsule, length=3, radius=1, shapeName=dyn_strap_B_02_capsule, center=<1 0 0>, rotation=<rot 90 0 0>}]
 dyn_strap_B_03 [{shapeType=Capsule, length=3, radius=1, shapeName=dyn_strap_B_03_capsule, center=<1.5 0 0>, rotation=<rot 90 0 0>}]
}
constraints [
 {
  bone1Name=base_strap_A
  bone2Name=Melee_TwoHanded_Handle
  swing1LimitDegrees=45
  swing1Motion=Free
  swing2LimitDegrees=45
  swing2Motion=Free
  twistLimitDegrees=45
  twistMotion=Free
 }
 {bone1Name=dyn_strap_A_01, bone2Name=base_strap_A, swing1LimitDegrees=20, swing2LimitDegrees=160, twistLimitDegrees=1}
 {bone1Name=dyn_strap_A_02, bone2Name=dyn_strap_A_01, swing1LimitDegrees=5, swing2LimitDegrees=25, twistLimitDegrees=1}
 {bone1Name=dyn_strap_A_03, bone2Name=dyn_strap_A_02, swing1LimitDegrees=10, swing2LimitDegrees=25, twistLimitDegrees=1}
 {
  bone1Name=base_strap_B
  bone2Name=Melee_TwoHanded_Handle
  swing1LimitDegrees=45
  swing1Motion=Free
  swing2LimitDegrees=45
  swing2Motion=Free
  twistLimitDegrees=45
  twistMotion=Free
 }
 {bone1Name=dyn_strap_B_01, bone2Name=base_strap_B, swing1LimitDegrees=20, swing2LimitDegrees=140, twistLimitDegrees=1}
 {bone1Name=dyn_strap_B_02, bone2Name=dyn_strap_B_01, swing1LimitDegrees=5, swing2LimitDegrees=25, twistLimitDegrees=1}
 {bone1Name=dyn_strap_B_03, bone2Name=dyn_strap_B_02, swing1LimitDegrees=10, swing2LimitDegrees=25, twistLimitDegrees=1}
]
