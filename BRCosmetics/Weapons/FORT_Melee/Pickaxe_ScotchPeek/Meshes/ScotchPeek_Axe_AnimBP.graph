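# Anim Blueprint for the ScotchPeek pickaxe that adds secondary motion to its two hanging straps.
# Nested bool blends choose by bIsInFrontEnd, then bRigidBodyEnabled: front end uses per-bone AnimDynamics when enabled, otherwise RigidBody with an override physics asset.
# Both physics branches read the cached input pose.
# When rigid body is disabled, either context falls back to planar-limited Trail controllers on pitched strap bases, tuned separately for each.
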AnimBlueprint /BRCosmetics/Weapons/FORT_Melee/Pickaxe_ScotchPeek/Meshes/ScotchPeek_Axe_AnimBP
target_skeleton=&/ScotchPeek_Axe_Skeleton
graphs [
 AnimGraph {nodes [
   n1 = "Transform (Modify) Bone - Bone: base_strap_A"(Alpha=1, Translation="0,0,0", Rotation="-90,0,0", Scale="1,1,1") {
    rotationMode=BMM_Additive
    rotationSpace=BCS_BoneSpace
    scaleMode=BMM_Additive
    scaleSpace=BCS_BoneSpace
    translationMode=BMM_Additive
    translationSpace=BCS_BoneSpace
    boneToModify {boneName=base_strap_A}
    rotation {pitch=-90}
   }
   n2 = Trail controller - Bone: dyn_strap_A_03(ComponentPose=@n1, Alpha=1) {
    bUsePlanarLimit=true
    chainLength=3
    maxDeltaTime=0.0333
    planarLimits [{drivingBone {boneName=base_strap_A}, planeTransform {location=<0 0 0>, rotation=<rot 90 180 0>, scale=<1 1 1>}}]
    rotationLimits [{limitMax=<180 180 180>}, {limitMax=<180 180 180>}, {limitMax=<180 180 180>}]
    rotationOffsets [<0 0 0>, <0 0 0>, <0 0 0>]
    trailBone {boneName=dyn_strap_A_03}
    alphaBoolBlend=~
   }
   n3 = "Transform (Modify) Bone - Bone: base_strap_B"(ComponentPose=@n2, Alpha=1, Translation="0,0,0", Rotation="-90,0,0", Scale="1,1,1") {
    rotationMode=BMM_Additive
    rotationSpace=BCS_BoneSpace
    scaleMode=BMM_Additive
    scaleSpace=BCS_BoneSpace
    translationMode=BMM_Additive
    translationSpace=BCS_BoneSpace
    boneToModify {boneName=base_strap_B}
    rotation {pitch=-90}
   }
   n4 = Trail controller - Bone: dyn_strap_B_03(ComponentPose=@n3, Alpha=1) {
    bUsePlanarLimit=true
    chainLength=3
    maxDeltaTime=0.0333
    planarLimits [{drivingBone {boneName=base_strap_B}, planeTransform {location=<0 0 0>, rotation=<rot 90 180 0>, scale=<1 1 1>}}]
    rotationLimits [{limitMax=<180 180 180>}, {limitMax=<180 180 180>}, {limitMax=<180 180 180>}]
    rotationOffsets [<0 0 0>, <0 0 0>, <0 0 0>]
    trailBone {boneName=dyn_strap_B_03}
    alphaBoolBlend=~
   }
   n5 = Component To Local(ComponentPose=@n4)
   n6 = Anim Dynamics - Bone: dyn_strap_A_01(ComponentPose=Local To Component(LocalPose=Use cached pose 'cache_pose_1'()), Alpha=0.7, ExternalForce="0,0,0") {
    angularDampingOverride=1
    angularSpringConstant=10000
    bOverrideAngularDamping=true
    bUseGravityOverride=true
    gravityScale=0.35
    linearSpringConstant=1000
    simulationSpace=RootRelative
    boundBone {boneName=dyn_strap_A_01}
    componentAppliedLinearAccClamp=<0 0 0>
    gravityOverride {z=-1500}
    physicsBodyDefinitions [{
      boundBone {boneName=dyn_strap_A_01}
      boxExtents=<4 6 6>
      constraintSetup {angularLimitsMax=<0 90 35>, angularLimitsMin=<0 -90 -35>, angularTarget=<1 0 0>}
      localJointOffset=<1 0 0>
     }]
    simSpaceSettings {maxAngularAcceleration=60, maxAngularVelocity=60, simSpaceAngularAlpha=1}
   }
   n7 = Anim Dynamics - Bone: dyn_strap_A_02(ComponentPose=@n6, Alpha=1, ExternalForce="0,0,0") {
    angularDampingOverride=1
    angularSpringConstant=10000
    bOverrideAngularDamping=true
    bUseGravityOverride=true
    gravityScale=0.35
    linearSpringConstant=1000
    simulationSpace=RootRelative
    boundBone {boneName=dyn_strap_A_02}
    componentAppliedLinearAccClamp=<0 0 0>
    gravityOverride {z=-1500}
    physicsBodyDefinitions [{
      boundBone {boneName=dyn_strap_A_02}
      boxExtents=<8 8 8>
      constraintSetup {angularLimitsMax=<0 90 35>, angularLimitsMin=<0 -90 -35>, angularTarget=<1 0 0>}
      localJointOffset=<1 0 0>
     }]
    simSpaceSettings {maxAngularAcceleration=60, maxAngularVelocity=60, simSpaceAngularAlpha=1}
   }
   n8 = Anim Dynamics - Bone: dyn_strap_A_03(ComponentPose=@n7, Alpha=1, ExternalForce="0,0,0") {
    angularDampingOverride=1
    angularSpringConstant=10000
    bOverrideAngularDamping=true
    bUseGravityOverride=true
    gravityScale=0.35
    linearSpringConstant=1000
    simulationSpace=RootRelative
    boundBone {boneName=dyn_strap_A_03}
    componentAppliedLinearAccClamp=<0 0 0>
    gravityOverride {z=-1500}
    physicsBodyDefinitions [{
      boundBone {boneName=dyn_strap_A_03}
      boxExtents=<8 8 8>
      constraintSetup {angularLimitsMax=<0 90 35>, angularLimitsMin=<0 -90 -35>, angularTarget=<1 0 0>}
      localJointOffset=<1 0 0>
     }]
    simSpaceSettings {maxAngularAcceleration=60, maxAngularVelocity=60, simSpaceAngularAlpha=1}
   }
   n9 = Anim Dynamics - Bone: dyn_strap_B_01(ComponentPose=@n8, Alpha=1, ExternalForce="0,0,0") {
    angularDampingOverride=1
    angularSpringConstant=10000
    bOverrideAngularDamping=true
    bUseGravityOverride=true
    gravityScale=0.35
    linearSpringConstant=1000
    simulationSpace=RootRelative
    boundBone {boneName=dyn_strap_B_01}
    componentAppliedLinearAccClamp=<0 0 0>
    gravityOverride {z=-1500}
    physicsBodyDefinitions [{
      boundBone {boneName=dyn_strap_B_01}
      boxExtents=<4 6 6>
      constraintSetup {angularLimitsMax=<0 90 35>, angularLimitsMin=<0 -90 -35>, angularTarget=<1 0 0>}
      localJointOffset=<1 0 0>
     }]
    simSpaceSettings {maxAngularAcceleration=60, maxAngularVelocity=60, simSpaceAngularAlpha=1}
   }
   n10 = Anim Dynamics - Bone: dyn_strap_B_02(ComponentPose=@n9, Alpha=1, ExternalForce="0,0,0") {
    angularDampingOverride=1
    angularSpringConstant=10000
    bOverrideAngularDamping=true
    bUseGravityOverride=true
    gravityScale=0.35
    linearSpringConstant=1000
    simulationSpace=RootRelative
    boundBone {boneName=dyn_strap_B_02}
    componentAppliedLinearAccClamp=<0 0 0>
    gravityOverride {z=-1500}
    physicsBodyDefinitions [{
      boundBone {boneName=dyn_strap_B_02}
      boxExtents=<8 8 8>
      constraintSetup {angularLimitsMax=<0 90 35>, angularLimitsMin=<0 -90 -35>, angularTarget=<1 0 0>}
      localJointOffset=<1 0 0>
     }]
    simSpaceSettings {maxAngularAcceleration=60, maxAngularVelocity=60, simSpaceAngularAlpha=1}
   }
   n11 = Anim Dynamics - Bone: dyn_strap_B_03(ComponentPose=@n10, Alpha=1, ExternalForce="0,0,0") {
    angularDampingOverride=1
    angularSpringConstant=10000
    bOverrideAngularDamping=true
    bUseGravityOverride=true
    gravityScale=0.35
    linearSpringConstant=1000
    simulationSpace=RootRelative
    boundBone {boneName=dyn_strap_B_03}
    componentAppliedLinearAccClamp=<0 0 0>
    gravityOverride {z=-1500}
    physicsBodyDefinitions [{
      boundBone {boneName=dyn_strap_B_03}
      boxExtents=<8 8 8>
      constraintSetup {angularLimitsMax=<0 90 35>, angularLimitsMin=<0 -90 -35>, angularTarget=<1 0 0>}
      localJointOffset=<1 0 0>
     }]
    simSpaceSettings {maxAngularAcceleration=60, maxAngularVelocity=60, simSpaceAngularAlpha=1}
   }
   n12 = Component To Local(ComponentPose=@n11)
   n13 = Get bRigidBodyEnabled()
   n14 = RigidBody!(ComponentPose=Local To Component(LocalPose=Use cached pose 'cache_pose_1'()), Alpha, OverrideWorldGravity:Vector="0,0,-980", ExternalForce) -> (Pose) {
    bClampLinearTranslationLimitToRefPose=true
    bOverrideWorldGravity=true
    overridePhysicsAsset=&/ScotchPeek_Axe_Physics
    simulationSpace=BaseBoneSpace
    baseBoneRef {boneName=Melee_TwoHanded_Handle}
    componentAppliedLinearAccClamp=<0 0 0>
    simSpaceSettings {
     dampingAlpha=0
     maxAngularAcceleration=15
     maxAngularVelocity=8
     maxLinearAcceleration=2000
     maxLinearVelocity=700
     worldAlpha=1
     externalLinearDragV=<0.5 0.5 0.5>
    }
   }
   n15 = Component To Local(ComponentPose=@n14)
   n16 = "Transform (Modify) Bone - Bone: base_strap_A"(Alpha=1, Translation="0,0,0", Rotation="-120,0,0", Scale="1,1,1") {
    rotationMode=BMM_Additive
    rotationSpace=BCS_BoneSpace
    scaleMode=BMM_Additive
    scaleSpace=BCS_BoneSpace
    translationMode=BMM_Additive
    translationSpace=BCS_BoneSpace
    boneToModify {boneName=base_strap_A}
    rotation {pitch=-120}
   }
   n17 = Trail controller - Bone: dyn_strap_A_03(ComponentPose=@n16, Alpha=0.75) {
    bUsePlanarLimit=true
    chainLength=3
    maxDeltaTime=0.0333
    planarLimits [{drivingBone {boneName=base_strap_A}, planeTransform {location=<0 0 0>, rotation=<rot 90 180 0>, scale=<1 1 1>}}]
    rotationLimits [{limitMax=<180 180 180>}, {limitMax=<180 180 180>}, {limitMax=<180 180 180>}]
    rotationOffsets [<0 0 0>, <0 0 0>, <0 0 0>]
    trailBone {boneName=dyn_strap_A_03}
    alphaBoolBlend=~
   }
   n18 = "Transform (Modify) Bone - Bone: base_strap_B"(ComponentPose=@n17, Alpha=1, Translation="0,0,0", Rotation="-120,0,0", Scale="1,1,1") {
    rotationMode=BMM_Additive
    rotationSpace=BCS_BoneSpace
    scaleMode=BMM_Additive
    scaleSpace=BCS_BoneSpace
    translationMode=BMM_Additive
    translationSpace=BCS_BoneSpace
    boneToModify {boneName=base_strap_B}
    rotation {pitch=-120}
   }
   n19 = Trail controller - Bone: dyn_strap_B_03(ComponentPose=@n18, Alpha=0.75) {
    bUsePlanarLimit=true
    chainLength=3
    maxDeltaTime=0.0333
    planarLimits [{drivingBone {boneName=base_strap_B}, planeTransform {location=<0 0 0>, rotation=<rot 90 180 0>, scale=<1 1 1>}}]
    rotationLimits [{limitMax=<180 180 180>}, {limitMax=<180 180 180>}, {limitMax=<180 180 180>}]
    rotationOffsets [<0 0 0>, <0 0 0>, <0 0 0>]
    trailBone {boneName=dyn_strap_B_03}
    alphaBoolBlend=~
   }
   n20 = Component To Local(ComponentPose=@n19)
   n21 = Get bRigidBodyEnabled()
   Output Pose(Result=Blend Poses by bool(
     BlendPose_0=Blend Poses by bool(BlendPose_0=@n12, BlendPose_1=@n5, bActiveValue=@n13)
     BlendPose_1=Blend Poses by bool(BlendPose_0=@n15, BlendPose_1=@n20, bActiveValue=@n21)
     bActiveValue=Get bIsInFrontEnd()
    )) {layerGroup=None}
   AnimGraphNode_SubInput_0 = Input Pose()
   Save cached pose 'cache_pose_1'(Pose=@AnimGraphNode_SubInput_0) {cacheName=cache_pose_1}
  ]}
 EventGraph {nodes [
   TryGetPawnOwner()
   Event BlueprintUpdateAnimation()
  ]}
]
variables []
